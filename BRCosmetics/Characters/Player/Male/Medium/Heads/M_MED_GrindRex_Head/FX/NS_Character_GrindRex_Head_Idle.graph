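# Niagara idle effect for the GrindRex character head cosmetic: glowing eyes with flame trails and soft flipbook wisps.
# Glow sprites and FlameWick puffs spawn at two mirrored eye points, faded by system HeadAlpha, with front-end adjustments.
# Each trail is a ten-point ribbon repositioned every update along shared EyeCurve, offset against DynamicVel to lag motion.
# System colors drive glow tint and trail flame gradient.
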NiagaraSystem /BRCosmetics/Characters/Player/Male/Medium/Heads/M_MED_GrindRex_Head/FX/NS_Character_GrindRex_Head_Idle
dependencies {
 usedDataInterfaces [
  &NiagaraDataInterfaceCurve
  &NiagaraDataInterfaceColorCurve
  &NiagaraDataInterfaceSpriteRendererInfo
  &NiagaraDataInterfaceCamera
  &NiagaraDataInterfaceVectorCurve
 ]
 usedDynamicInputs [
  &MergeHelper_Vector2Switch
  &Multiply_Vector2D_ByFloat
  &RandomRangeFloat
  &TransformPosition
  &Multiply_Float
  &FNTimeOfDayColor
  &RemapRange
  &DotProduct
  &ReturnCameraVector
  &MakeVector4FromVectorAndFloat
  &MakeLinearColorFromVectorAndFloat
  /Niagara/DynamicInputs/Math/Clamp_Vector
 ]
 usedModules [
  &SystemState
  &self:SystemSpawnScript.SystemScriptSource.SystemScriptGraph.NiagaraNodeAssignment_0.SetVariables_#1
  &self:SystemSpawnScript.SystemScriptSource.SystemScriptGraph.NiagaraNodeAssignment_2.SetVariables_#2
  &self:SystemSpawnScript.SystemScriptSource.SystemScriptGraph.NiagaraNodeAssignment_9.SetVariables_#3
  &InterpolateOverTime
  &self:SystemSpawnScript.SystemScriptSource.SystemScriptGraph.NiagaraNodeAssignment_11.SetVariables_#4
 ]
}
emitters {
 Eyes_Arm_Glow {
  emitterUpdateScript [
   &EmitterState(Life Cycle Mode=::NewEnumerator1, Loop Behavior=::NewEnumerator1, Loop Duration Mode=::NewEnumerator1, UseLoopDelay=false)
   &SpawnBurst_Instantaneous(Spawn Count=2, Use Spawn Probability=false, UseLoopCountLimit=true)
  ]
  particleSpawnScript [
   &PartitionParticles(ID to Partition:NiagaraInt32=&ReturnExecIndex())
   &InitializeParticle(
    Color=<rgba 3 0.4924 0.0117 2>
    Position=&MergeHelper_VectorSwitch(0=<6 7 9>, 1=<-6 7 9>, 2=?, 3=?, Index=$Particles.PartitionParticles.Partition, Num=::NewEnumerator0)
    Position Mode=::NewEnumerator1
    Sprite Size Mode=::NewEnumerator3
    Uniform Sprite Size=9
    UsePositionOffset=false
    Write Material Random=true
   )
   &CameraOffset(Camera Offset Amount=&FN_FrontEndFloatModifier(Gameplay=0.5, Front End=4))
  ]
  particleUpdateScript [
   &ParticleState(Kill Particles When Lifetime Has Elapsed=false, Let Infinitely Lived Particles Die When Emitter Deactivates=true)
   &DynamicMaterialParameters(
    Index 0 Param 1=1
    Index 0 Param 3=5
    Module.Index0Param0Write=true
    Module.Index0Param1Write=false
    Module.Index0Param2Write=false
    Module.Index0Param3Write=false
   )
   &SpriteFacingAndAlignment(
    Sprite Alignment=&MergeHelper_VectorSwitch(
     0=&ReturnMeshOrientationAxis(Mesh Orientation Axis=::XAxis, Mesh Orientation Quat=$Particles.MeshOrientation)
     1=&RotateVectorByQuaternion(Vector To Rotate=<1 0 0>, Quaternion=$Particles.SkeletalMeshLocation.SampledRotation)
     2=?
     3=?
     Index=$Particles.EyesOrArm
     Num=::NewEnumerator0
    )
    Sprite Facing=&MergeHelper_VectorSwitch(
     0=&ReturnMeshOrientationAxis(Mesh Orientation Axis=::YAxis, Mesh Orientation Quat=$Particles.MeshOrientation)
     1=<0 0 1>
     2=?
     3=?
     Index=$Particles.EyesOrArm
     Num=::NewEnumerator0
    )
    WriteAlignment=true
    WriteFacing=true
   )
   &ScaleColor(Scale Alpha=$System.HeadAlpha, ScaleA=true, ScaleRGB=false)
   &ScaleColor(Color Scale:LinearColor=$System.TOD Color, Color Value To Scale=$Particles.Color, Scale Mode=3) {module=ScaleColor001}
   &ScaleColor(
    Scale Alpha=&MergeHelper_FloatSwitch(
     0=&FN_FrontEndFloatModifier(Gameplay=0.8, Front End=0.4)
     1=&FNTimeOfDayScalar(Night Scale=1.3, Morning Scale=1.3, Evening Scale=1)
     2=?
     3=?
     Index=$Particles.EyesOrArm
     Num=::NewEnumerator0
    )
    ScaleA=true
    ScaleRGB=false
   ) {module=ScaleColor002}
  ]
  rendererClasses [&NiagaraSpriteRendererProperties]
  renderers [{rendererClass=&NiagaraSpriteRendererProperties}]
 }
 FlameTrail_L {
  emitterUpdateScript [
   &EmitterState(Life Cycle Mode=::NewEnumerator1, Loop Behavior=::NewEnumerator1, Loop Duration Mode=::NewEnumerator1, UseLoopDelay=false)
   &SpawnBurst_Instantaneous(Spawn Count=10, Use Spawn Probability=false, UseLoopCountLimit=false)
  ]
  particleSpawnScript [
   &InitializeParticle(
    Color Mode=::NewEnumerator0
    Lifetime=0.3
    Position Mode=::NewEnumerator0
    Ribbon Width=4
    Ribbon Width Mode=::NewEnumerator1
    Sprite Size Mode=::NewEnumerator3
    Uniform Sprite Size=50
    Write Material Random=true
   )
   /SetVariables(Particles.NormalizedIndex:NiagaraFloat=&Divide_Float(
     A=&MakeFloatFromInt(Int=$Engine.ExecIndex)
     B=&MakeFloatFromInt(Int=/Niagara/DynamicInputs/Subtract/Subtract_Integer(A:NiagaraInt32=$Engine.ExecutionCount, B:NiagaraInt32=1))
    )) {isSetParameters=true, module=SetVariables_#5}
  ]
  particleUpdateScript [
   &ParticleState(Kill Particles When Lifetime Has Elapsed=false)
   /SetVariables(Particles.Position=&AddVectorToPosition(
     Position=<6 7 9>
     Vector=&Add_Vector(
      A=&VectorFromCurve(
       VectorCurve=$System.EyeCurve
       CurveIndex=&OneMinusFloat(Float=$Particles.Initial.NormalizedIndex)
       Scale Curve=<2.9 -20 7.9>
      )
      B=&TransformVector(
       Vector=&Multiply_VectorByFloat(
        Vector=$System.DynamicVel
        Float=&FloatFromCurve(
         FloatCurve=DataInterface() {propertyValues {Curve {keys [
             {arriveTangent=-0.428, interpMode=RCIM_Cubic, leaveTangent=-0.428, tangentMode=RCTM_User, time=0, value=1}
             {arriveTangent=-1, interpMode=RCIM_Cubic, leaveTangent=-1, time=0.4509, value=0.6265}
             <at 1 0 RCIM_Cubic>
            ]}}}
         CurveIndex=$Particles.Initial.NormalizedIndex
         Scale Curve=-0.04
        )
       )
       Source Space=::World
       Destination Space=::Local
      )
     )
    )) {isSetParameters=true, module=SetVariables_#6}
   /SetVariables(
    Particles.LocalFlameA:LinearColor=$System.FlameA
    Particles.LocalFlameB:LinearColor=&MultiplyLinearColor(
     Linear Color=$System.FlameB
     Scale Factor=&MakeVector4_FromLinearColor(LinearColor=&ColorFromCurve(
       ColorCurve=DataInterface() {propertyValues {
         AlphaCurve {keys [<at 0 1 RCIM_Cubic>,<at 0.4267 1>,<at 0.8175 0.34>]}
         BlueCurve {keys [<at 0 0 RCIM_Cubic>,<at 0.1708 0>,<at 0.265 0 RCIM_Cubic>,<at 0.6474 0>]}
         GreenCurve {keys [<at 0 2.2473 RCIM_Cubic>, <at 0.1708 0.9577>, {interpMode=RCIM_Cubic, leaveTangent=-2.7331, time=0.265, value=0.433}, <at 0.6474 0.2513>]}
         RedCurve {keys [<at 0 3 RCIM_Cubic>, <at 0.1708 1.5784>, {interpMode=RCIM_Cubic, leaveTangent=-2.7386, time=0.265, value=1}, <at 0.6474 1>]}
        }}
       CurveIndex=&OneMinusFloat(Float=$Particles.Initial.NormalizedIndex)
      ))
    )
   ) {isSetParameters=true, module=SetVariables_#7}
  ]
  rendererClasses [&NiagaraRibbonRendererProperties]
  renderers [
   {rendererClass=&NiagaraRibbonRendererProperties}
   {rendererClass=&NiagaraRibbonRendererProperties, rendererIndex=1}
  ]
 }
 FlameTrail_R {
  emitterUpdateScript [
   &EmitterState(Life Cycle Mode=::NewEnumerator1, Loop Behavior=::NewEnumerator1, Loop Duration Mode=::NewEnumerator1, UseLoopDelay=false)
   &SpawnBurst_Instantaneous(Spawn Count=10, Use Spawn Probability=false, UseLoopCountLimit=false)
  ]
  particleSpawnScript [
   &InitializeParticle(
    Color Mode=::NewEnumerator0
    Lifetime=0.3
    Position Mode=::NewEnumerator0
    Ribbon Width=4
    Ribbon Width Mode=::NewEnumerator1
    Sprite Size Mode=::NewEnumerator3
    Uniform Sprite Size=50
    Write Material Random=true
   )
   /SetVariables(Particles.NormalizedIndex:NiagaraFloat=&Divide_Float(
     A=&MakeFloatFromInt(Int=$Engine.ExecIndex)
     B=&MakeFloatFromInt(Int=&Subtract_Integer(A:NiagaraInt32=$Engine.ExecutionCount, B:NiagaraInt32=1))
    )) {isSetParameters=true, module=SetVariables_#5}
  ]
  particleUpdateScript [
   &ParticleState(Kill Particles When Lifetime Has Elapsed=false)
   /SetVariables(Particles.Position=&AddVectorToPosition(
     Position=<-6 7 9>
     Vector=&Add_Vector(
      A=&VectorFromCurve(
       VectorCurve=$System.EyeCurve
       CurveIndex=&OneMinusFloat(Float=$Particles.Initial.NormalizedIndex)
       Scale Curve=<-2.9 -20 7.9>
      )
      B=&TransformVector(
       Vector=&Multiply_VectorByFloat(
        Vector=$System.DynamicVel
        Float=&FloatFromCurve(
         FloatCurve=DataInterface() {propertyValues {Curve {keys [
             {arriveTangent=-0.428, interpMode=RCIM_Cubic, leaveTangent=-0.428, tangentMode=RCTM_User, time=0, value=1}
             {arriveTangent=-1, interpMode=RCIM_Cubic, leaveTangent=-1, time=0.4509, value=0.6265}
             <at 1 0 RCIM_Cubic>
            ]}}}
         CurveIndex=$Particles.Initial.NormalizedIndex
         Scale Curve=-0.04
        )
       )
       Source Space=::World
       Destination Space=::Local
      )
     )
    )) {isSetParameters=true, module=SetVariables_#6}
   /SetVariables(
    Particles.LocalFlameA:LinearColor=$System.FlameA
    Particles.LocalFlameB:LinearColor=&MultiplyLinearColor(
     Linear Color=$System.FlameB
     Scale Factor=&MakeVector4_FromLinearColor(LinearColor=&ColorFromCurve(
       ColorCurve=DataInterface() {propertyValues {
         AlphaCurve {keys [<at 0 1 RCIM_Cubic>,<at 0.4267 1>,<at 0.8175 0.34>]}
         BlueCurve {keys [<at 0 0 RCIM_Cubic>,<at 0.1708 0>,<at 0.265 0 RCIM_Cubic>,<at 0.6474 0>]}
         GreenCurve {keys [<at 0 2.2473 RCIM_Cubic>, <at 0.1708 0.9577>, {interpMode=RCIM_Cubic, leaveTangent=-2.7331, time=0.265, value=0.433}, <at 0.6474 0.2513>]}
         RedCurve {keys [<at 0 3 RCIM_Cubic>, <at 0.1708 1.5784>, {interpMode=RCIM_Cubic, leaveTangent=-2.7386, time=0.265, value=1}, <at 0.6474 1>]}
        }}
       CurveIndex=&OneMinusFloat(Float=$Particles.Initial.NormalizedIndex)
      ))
    )
   ) {isSetParameters=true, module=SetVariables_#7}
  ]
  rendererClasses [&NiagaraRibbonRendererProperties]
  renderers [
   {rendererClass=&NiagaraRibbonRendererProperties}
   {rendererClass=&NiagaraRibbonRendererProperties, rendererIndex=1}
  ]
 }
 FlameWick {
  emitterUpdateScript [
   &EmitterState()
   &SpawnRate(SpawnRate=3, Use Spawn Probability=false)
  ]
  particleSpawnScript [
   &FN_SetSpawnGroup(Number of Spawn Groups=2)
   &InitializeParticle(
    Color=<rgba 0.313 0.2848 0.1878 1>
    Lifetime Max=3.5
    Lifetime Min=2.5
    Lifetime Mode=::NewEnumerator1
    Position=&MergeHelper_VectorSwitch(0=<6 7 9>, 1=<-6 7 9>, 2=?, 3=?, Index=&SpawnGroup(), Num=::NewEnumerator0)
    Position Mode=::NewEnumerator1
    Sprite Rotation Mode=::NewEnumerator2
    Sprite Size Mode=::NewEnumerator1
    Uniform Sprite Size Max=6
    Uniform Sprite Size Min=4
    UsePositionOffset=false
    Write Material Random=true
   )
   &CameraOffset(Camera Offset Amount=&FN_FrontEndFloatModifier(Gameplay=0, Front End=2))
  ]
  particleUpdateScript [
   &ParticleState()
   &ScaleSpriteSize(
    Uniform Curve Scale=&MakeCustomFloatFromBool(Bool=$User.bIsFrontEndPreview, True Float=0.8, False Float=1.1)
    Uniform Curve Sprite Scale=DataInterface() {propertyValues {Curve {keys [
        {arriveTangent=5.6506, interpMode=RCIM_Cubic, leaveTangent=5.6506, tangentMode=RCTM_User, time=0, value=0.5}
        {arriveTangent=0.6355, interpMode=RCIM_Cubic, leaveTangent=0.6355, tangentMode=RCTM_User, time=0.5735, value=1.75}
       ]}}}
   )
   &ScaleColor(
    Scale Alpha=&FloatFromCurve(
     FloatCurve=DataInterface() {propertyValues {Curve {keys [
         <at 0 0 RCIM_Cubic>
         {arriveTangent=0.0415, interpMode=RCIM_Cubic, leaveTangent=0.0415, tangentMode=RCTM_User, time=0.2384, value=1}
         {arriveTangent=-2.099, interpMode=RCIM_Cubic, leaveTangent=-2.099, tangentMode=RCTM_User, time=1, value=0}
        ]}}}
     Scale Curve=$System.HeadAlpha
    )
    Scale RGB=&MakeVectorFromLinearColorRGB(Color=&ColorFromCurve(ColorCurve=DataInterface() {propertyValues {
        AlphaCurve {keys []}
        BlueCurve {keys [<at 0.1682 1 RCIM_Cubic>,<at 1 0.2492 RCIM_Cubic>]}
        GreenCurve {keys [<at 0.1682 1 RCIM_Cubic>,<at 1 0.3619 RCIM_Cubic>]}
        RedCurve {keys [<at 0.1682 1 RCIM_Cubic>,<at 1 1 RCIM_Cubic>]}
       }}))
    ScaleA=true
    ScaleRGB=true
   )
   &SubUVAnimation(
    End Frame Range Override=16
    "Sprite Renderer "=DataInterface() {propertyValues {SpriteRenderer=&self:FlameWick_0.NiagaraSpriteRendererProperties_2}}
    Start Frame Range Override=8
    SubUV Animation Mode:ENiagara_SubUVLookupModeV2=::NewEnumerator1
    SubUV Curve=DataInterface() {propertyValues {Curve {
       defaultValue=0
       keys [
        {arriveTangent=-0.1034, interpMode=RCIM_Cubic, leaveTangent=-0.1034, tangentMode=RCTM_User, time=0, value=0}
        {arriveTangent=1, interpMode=RCIM_Cubic, leaveTangent=1, time=0.6303, value=0.8441}
        <at 1 1 RCIM_Cubic>
       ]
      }}}
    UseEndFrame=true
    UseStartFrame=true
   )
  ]
  rendererClasses [&NiagaraSpriteRendererProperties]
  renderers [{rendererClass=&NiagaraSpriteRendererProperties}]
 }
}
summary {
 systemName=NS_Character_GrindRex_Head_Idle
 emitters [
  {emitterName=Eyes_Arm_Glow, rendererClasses [&NiagaraSpriteRendererProperties]}
  {emitterName=FlameWick, rendererClasses [&NiagaraSpriteRendererProperties]}
  {emitterName=FlameTrail_L, rendererClasses [&NiagaraRibbonRendererProperties]}
  {emitterName=FlameTrail_R, rendererClasses [&NiagaraRibbonRendererProperties]}
 ]
 userVariables [
  User.bIsADS {type=NiagaraBool}
  User.bIsFrontEnd {type=NiagaraBool}
  User.bIsFrontEndPreview {type=NiagaraBool}
  User.bIsHeadHidden {type=NiagaraBool}
  User.bisSwinging {type=NiagaraBool}
  User.Variant {type=NiagaraFloat, defaultValue=0, description=""}
 ]
}
