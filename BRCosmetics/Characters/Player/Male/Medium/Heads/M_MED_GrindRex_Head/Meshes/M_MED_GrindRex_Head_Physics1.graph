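# Physics asset for the GrindRex male medium player head cosmetic, setting up a dynamic tongue attached to the jaw.
# Capsule bodies cover C_jaw and a three-bone tongue chain of dyn_tongue_base, dyn_tongue_01 and dyn_tongue_02.
# The base's constraint to the jaw leaves swing and twist free, so its angle limits are inactive.
# The two outer joints have tighter angle limits, with the tip allowed more bend and twist than the middle.
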PhysicsAsset /BRCosmetics/Characters/Player/Male/Medium/Heads/M_MED_GrindRex_Head/Meshes/M_MED_GrindRex_Head_Physics1
bodies {
 C_jaw [{shapeType=Capsule, length=18.8411, radius=4.2593, shapeName=C_jaw_capsule, center=<0.1238 8.0911 -6.0065>, rotation=<rot 0 -0.9451 -69.4744>}]
 dyn_tongue_01 [{shapeType=Capsule, length=2.5319, radius=2.1124, shapeName=dyn_tongue_01_capsule, center=<1.6622 0 0>, rotation=<rot 90 0 0>}]
 dyn_tongue_02 [{shapeType=Capsule, length=3.7996, radius=2.0891, shapeName=dyn_tongue_02_capsule, center=<1.6622 0 0>, rotation=<rot 90 0 0>}]
 dyn_tongue_base [{shapeType=Capsule, length=0.505, radius=0.505, shapeName=dyn_tongue_base_capsule, rotation=<rot 0 -90 0>}]
}
constraints [
 {
  bone1Name=dyn_tongue_base
  bone2Name=C_jaw
  swing1LimitDegrees=45
  swing1Motion=Free
  swing2LimitDegrees=45
  swing2Motion=Free
  twistLimitDegrees=45
  twistMotion=Free
 }
 {bone1Name=dyn_tongue_01, bone2Name=dyn_tongue_base, swing1LimitDegrees=20, swing2LimitDegrees=20, twistLimitDegrees=10}
 {bone1Name=dyn_tongue_02, bone2Name=dyn_tongue_01, swing1LimitDegrees=30, swing2LimitDegrees=30, twistLimitDegrees=20}
]
